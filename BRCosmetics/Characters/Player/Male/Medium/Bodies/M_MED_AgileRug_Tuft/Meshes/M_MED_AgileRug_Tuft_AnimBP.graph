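# Animation Blueprint for the M_MED_AgileRug_Tuft player body cosmetic, copying its pose from the attached parent mesh and layering costume-specific corrections on top.
# The copied pose feeds a Control Rig, which receives a crouch-walking value that each animation update derives from crouching AND 2D movement, cast to a number.
# Two additive bone-space modifiers then apply fixed mirrored translation offsets and a matching pitch to the left and right dynamic shoulder pad bones.
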AnimBlueprint /BRCosmetics/Characters/Player/Male/Medium/Bodies/M_MED_AgileRug_Tuft/Meshes/M_MED_AgileRug_Tuft_AnimBP
target_skeleton=&/M_MED_AgileRug_Tuft_Skeleton
graphs [
 AnimGraph {nodes [
   n1 = Control Rig!(Source=Copy Pose From Mesh() {bUseAttachedParent=true}, Alpha, is Crouch Moving:double=Get Is Crouch Walking() -> (Is Crouch Walking:double)) -> (Pose) {
    alphaBoolBlend {blendOption=Linear}
    controlRigAssetReference {blueprintRigClass=&/M_MED_AgileRug_Tuft_CtrlRig.M_MED_AgileRug_Tuft_CtrlRig_C}
   }
   n2 = "Transform (Modify) Bone - Bone: dyn_shoulderPad_l"(ComponentPose=Local To Component(LocalPose=@n1), Alpha=1, Translation="4.5,1,0", Rotation="8,0,0", Scale="1,1,1") {
    rotationMode=BMM_Additive
    rotationSpace=BCS_BoneSpace
    translationMode=BMM_Additive
    translationSpace=BCS_BoneSpace
    alphaBoolBlend {blendOption=Linear}
    alphaScaleBiasClamp {bInterpResult=true}
    boneToModify {boneName=dyn_shoulderPad_l}
    rotation {pitch=8}
    translation=<4.5 1>
   }
   n3 = "Transform (Modify) Bone - Bone: dyn_shoulderPad_r"(ComponentPose=@n2, Alpha=1, Translation="-4.5,-1,0", Rotation="8,0,0", Scale="1,1,1") {
    rotationMode=BMM_Additive
    rotationSpace=BCS_BoneSpace
    translationMode=BMM_Additive
    translationSpace=BCS_BoneSpace
    alphaBoolBlend {blendOption=Linear}
    alphaScaleBiasClamp {bInterpResult=true}
    boneToModify {boneName=dyn_shoulderPad_r}
    rotation {pitch=8}
    translation=<-4.5 -1>
   }
   Output Pose(Result=Component To Local(ComponentPose=@n3)) {layerGroup=None}
  ]}
 EventGraph {nodes [
   Event BlueprintUpdateAnimation()
   n1 = Get bIsCrouching()
   n2 = Get bIsMoving2D()
   > Set Is Crouch Walking(Is Crouch Walking:double="To Float (Boolean)"(InBool=AND Boolean(A=@n1, B=@n2))) -> (Output_Get:double)
  ]}
]
variables [Is Crouch Walking {type=number}]
